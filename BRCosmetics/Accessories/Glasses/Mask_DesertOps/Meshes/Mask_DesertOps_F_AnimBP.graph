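# Anim Blueprint for the female Desert Ops mask accessory, compensating for hair clipping on one specific character head.
# It copies the pose from the attached parent mesh, then applies an additive scale to the right earpiece bone, weighted by an alpha variable.
# When character customization loads, that alpha becomes 1 if the pawn's head mesh name contains the Emmy head identifier, otherwise 0.
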AnimBlueprint /BRCosmetics/Accessories/Glasses/Mask_DesertOps/Meshes/Mask_DesertOps_F_AnimBP
target_skeleton=&/Mask_DesertOps_F_Skeleton
graphs [
 AnimGraph {nodes [
   n1 = "Transform (Modify) Bone - Bone: desert_ops_glasses_right_earpiece"(
    ComponentPose=Local To Component(LocalPose=Copy Pose From Mesh() {bUseAttachedParent=true})
    Alpha=Get Hair Clipping Compensation Alpha() -> (Hair Clipping Compensation Alpha:double)
    Translation="0,0,0"
    Rotation="0,0,0"
    Scale="0.5,0.5,0.5"
   ) {
    scaleMode=BMM_Additive
    translationMode=BMM_Additive
    alphaBoolBlend {blendOption=Linear}
    boneToModify {boneName=desert_ops_glasses_right_earpiece}
    scale=<0.5 0.5 0.5>
   }
   Output Pose(Result=Component To Local(ComponentPose=@n1)) {layerGroup=None}
  ]}
 EventGraph {nodes [
   Event OnCharacterCustomizationLoaded()
   > n1 = Cast To PlayerPawn_Athena(Object=TryGetPawnOwner())
   n2 = GetDisplayName(Object=GetSkeletalMeshForPartType(self=@n1, PartType=Head))
   > Set Hair Clipping Compensation Alpha(Hair Clipping Compensation Alpha:double="To Float (Boolean)"(InBool=Contains(SearchIn=@n2, Substring=F_MED_CAU_Emmy_Head))) -> (Output_Get:double)
  ]}
]
variables [Hair Clipping Compensation Alpha {type=number}]
